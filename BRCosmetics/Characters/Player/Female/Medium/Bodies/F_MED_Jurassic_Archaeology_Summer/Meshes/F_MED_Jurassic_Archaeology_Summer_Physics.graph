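# Physics asset for the female medium Jurassic Archaeology Summer body cosmetic, covering its dangling bows, necklace and ropes.
# Capsules on pelvis, spine and left hand form the body base for constrained chains of small capsules.
# Bow A pair and necklace hang from spine_05, bow B pair from spine_01 (numbered 03 to 02), ropes from hand_l.
# Necklace limits allow almost only single-axis swing, tightening down the chain.
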PhysicsAsset /BRCosmetics/Characters/Player/Female/Medium/Bodies/F_MED_Jurassic_Archaeology_Summer/Meshes/F_MED_Jurassic_Archaeology_Summer_Physics
bodies {
 dny_bowA_l_01 [{shapeType=Capsule, length=4.5678, radius=0.646, center=<-2.104 0.0514 -0.131>, rotation=<rot 1.3881 -91.2249 -88.3345>}]
 dny_bowA_l_02 [{shapeType=Capsule, length=4.2432, radius=0.5, center=<-0.5638 0.251 -0.0798>, rotation=<rot -1.4404 -105.2031 -91.7049>}]
 dny_bowA_r_01 [{shapeType=Capsule, length=4.7506, radius=0.6472, center=<-2.0453 -0.088 -0.1196>, rotation=<rot -0.5201 91.7461 -89.2005>}]
 dny_bowA_r_02 [{shapeType=Capsule, length=4.1424, radius=0.5, center=<-0.7716 0.1434 0.027>, rotation=<rot 0 -105.0409 -86.4735>}]
 dny_bowB_l_02 [{shapeType=Capsule, length=1.7108, radius=0.9269, center=<0.1246 0 0.0039>, rotation=<rot 0 -89.9999 88.1987>}]
 dny_bowB_l_03 [{shapeType=Capsule, length=3.173, radius=0.9269, center=<-1.2488 0 -0.1068>, rotation=<rot 0 -89.9999 -94.8883>}]
 dny_bowB_r_02 [{shapeType=Capsule, length=1.7108, radius=0.9269, center=<0.1246 0 0.0039>, rotation=<rot 0 -89.9999 88.195>}]
 dny_bowB_r_03 [{shapeType=Capsule, length=3.1229, radius=0.9269, center=<-1.2552 0 -0.0715>, rotation=<rot 0 -89.9998 -93.26>}]
 dyn_necklace01 [{shapeType=Capsule, length=5.2778, radius=0.8136, center=<-2.426 0.0003 -0.0455>, rotation=<rot 88.8931 179.9957 179.9958>}]
 dyn_necklace02 [{shapeType=Capsule, length=3.4049, radius=1.2273, center=<-2.1719 -0.0001 -0.027>, rotation=<rot 89.4128 179.9995 179.9993>}]
 dyn_necklace03 [{shapeType=Capsule, length=1.0091, radius=0.6866, center=<-0.4579 0.0001 0.0033>, rotation=<rot 89.5817 -0.001 -0.0011>}]
 dyn_necklace04 [{shapeType=Capsule, length=4.8298, radius=0.9871, center=<-2.1594 -0.0196 0.0038>, rotation=<rot 0.0108 -86.8033 -89.7319>}]
 dyn_ropeA01 [{shapeType=Capsule, length=3.0026, radius=0.505, center=<-0.0425 -0.4913 0.1932>, rotation=<rot 0 177.2643 -81.6906>}]
 dyn_ropeA02 [{shapeType=Capsule, length=2.3723, radius=0.5, center=<0.1454 -1.0356 0.0264>, rotation=<rot 0 -168.788 -83.6428>}]
 dyn_ropeB01 [{shapeType=Capsule, length=2.3444, radius=0.6948, center=<-0.0658 -0.1979 -0.1376>, rotation=<rot 88.1485 -140.2309 -51.6569>}]
 dyn_ropeB02 [{shapeType=Capsule, length=2.136, radius=0.5, center=<-0.0232 -0.7619 -0.1286>, rotation=<rot 0 -173.9122 -81.6464>}]
 hand_l [{shapeType=Capsule, length=2.3233, radius=4.5237, center=<4.7368 -0.6578 -0.0943>, rotation=<rot 5.9174 -90.2278 86.4656>}]
 pelvis [{shapeType=Capsule, length=12.8416, radius=9.7713, center=<4.5338 -0.5903 -0.2195>, rotation=<rot -2.4417 -90 2.0417>}]
 spine_01 [{shapeType=Capsule, length=3.8798, radius=7.6475, center=<8.8479 2.0168 0.0723>, rotation=<rot 0 -6.4479 -177.0717>}]
 spine_03 [{shapeType=Capsule, length=9.3793, radius=8.5247, center=<5.1254 5.5073 0.3517>, rotation=<rot 0.0855 -90.4754 -2.9988>}]
 spine_05 [
  {shapeType=Capsule, length=11.1046, radius=8.6558, center=<-1.8002 1.0164 -0.0085>, rotation=<rot 0 40.5822 -179.927>}
  {shapeType=Capsule, length=16.7461, radius=5.7297, center=<5.2462 -2.1692 0>}
 ]
}
constraints [
 {bone1Name=spine_01, bone2Name=pelvis, swing1LimitDegrees=45, swing2LimitDegrees=45, twistLimitDegrees=45}
 {bone1Name=spine_03, bone2Name=spine_01, swing1LimitDegrees=45, swing2LimitDegrees=45, twistLimitDegrees=45}
 {bone1Name=spine_05, bone2Name=spine_03, swing1LimitDegrees=45, swing2LimitDegrees=45, twistLimitDegrees=45}
 {bone1Name=dny_bowA_r_01, bone2Name=spine_05, swing1LimitDegrees=40, swing2LimitDegrees=10, twistLimitDegrees=0.5}
 {bone1Name=dny_bowA_r_02, bone2Name=dny_bowA_r_01, swing1LimitDegrees=40, swing2LimitDegrees=5, twistLimitDegrees=0.5}
 {bone1Name=dny_bowA_l_01, bone2Name=spine_05, swing1LimitDegrees=40, swing2LimitDegrees=10, twistLimitDegrees=0.5}
 {bone1Name=dny_bowA_l_02, bone2Name=dny_bowA_l_01, swing1LimitDegrees=40, swing2LimitDegrees=5, twistLimitDegrees=0.5}
 {bone1Name=dyn_necklace01, bone2Name=spine_05, swing1LimitDegrees=30, swing2LimitDegrees=0, twistLimitDegrees=0}
 {bone1Name=dyn_necklace02, bone2Name=dyn_necklace01, swing1LimitDegrees=20, swing2LimitDegrees=0, twistLimitDegrees=0}
 {bone1Name=dyn_necklace03, bone2Name=dyn_necklace02, swing1LimitDegrees=10, swing2LimitDegrees=0, twistLimitDegrees=0}
 {
  bone1Name=dyn_necklace04
  bone2Name=dyn_necklace03
  swing1LimitDegrees=10
  swing2LimitDegrees=1
  swing2Motion=Locked
  twistLimitDegrees=1
  twistMotion=Locked
 }
 {bone1Name=dyn_ropeA01, bone2Name=hand_l, swing1LimitDegrees=20, swing2LimitDegrees=1, twistLimitDegrees=20}
 {bone1Name=dyn_ropeA02, bone2Name=dyn_ropeA01, swing1LimitDegrees=45, swing2LimitDegrees=1, twistLimitDegrees=45}
 {bone1Name=dyn_ropeB01, bone2Name=hand_l, swing1LimitDegrees=60, swing2LimitDegrees=60, twistLimitDegrees=60}
 {bone1Name=dyn_ropeB02, bone2Name=dyn_ropeB01, swing1LimitDegrees=60, swing2LimitDegrees=60, twistLimitDegrees=60}
 {bone1Name=dny_bowB_l_03, bone2Name=spine_01, swing1LimitDegrees=90, swing2LimitDegrees=30, twistLimitDegrees=20}
 {bone1Name=dny_bowB_l_02, bone2Name=dny_bowB_l_03, swing1LimitDegrees=90, swing2LimitDegrees=30, twistLimitDegrees=20}
 {bone1Name=dny_bowB_r_03, bone2Name=spine_01, swing1LimitDegrees=90, swing2LimitDegrees=30, twistLimitDegrees=20}
 {bone1Name=dny_bowB_r_02, bone2Name=dny_bowB_r_03, swing1LimitDegrees=90, swing2LimitDegrees=30, twistLimitDegrees=20}
]
