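# Physics asset for the F_MED_Jurassic_Archaeology_Summer female body cosmetic, covering two dangling rope pieces at the left hand.
# Small capsules on two two-bone chains, dyn_ropeA and dyn_ropeB, are each constrained to a larger hand_l capsule, then link bone to bone.
# Chain A has tight, nearly planar limits with one narrow swing axis, while chain B swings and twists more freely on all axes.
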PhysicsAsset /BRCosmetics/Characters/Player/Female/Medium/Bodies/F_MED_Jurassic_Archaeology_Summer/Meshes/F_MED_Jurassic_Archaeology_Summer_Physics1
bodies {
 dyn_ropeA01 [{shapeType=Capsule, length=1.6526, radius=0.505, center=<-0.0595 -0.8448 0.1413>, rotation=<rot 0.0001 177.2643 -94.8611>}]
 dyn_ropeA02 [{shapeType=Capsule, length=1.1708, radius=0.5, center=<0.1454 -1.0356 0.0264>, rotation=<rot 0 -168.788 -83.6428>}]
 dyn_ropeB01 [{shapeType=Capsule, length=1.7873, radius=0.505, center=<0.0107 -0.5888 0.0304>, rotation=<rot 72.5787 -166.8738 -77.8252>}]
 dyn_ropeB02 [{shapeType=Capsule, length=1.8837, radius=0.5, center=<0.0672 -0.6559 -0.0838>, rotation=<rot 0 -173.9122 -83.1826>}]
 hand_l [{shapeType=Capsule, length=4.2487, radius=4.6034, center=<4.9469 -0.0001 -0.3982>, rotation=<rot -85.3977 179.9998 -179.9998>}]
}
constraints [
 {bone1Name=dyn_ropeA01, bone2Name=hand_l, swing1LimitDegrees=20, swing2LimitDegrees=1, twistLimitDegrees=20}
 {bone1Name=dyn_ropeA02, bone2Name=dyn_ropeA01, swing1LimitDegrees=45, swing2LimitDegrees=1, twistLimitDegrees=45}
 {bone1Name=dyn_ropeB01, bone2Name=hand_l, swing1LimitDegrees=60, swing2LimitDegrees=60, twistLimitDegrees=60}
 {bone1Name=dyn_ropeB02, bone2Name=dyn_ropeB01, swing1LimitDegrees=60, swing2LimitDegrees=60, twistLimitDegrees=60}
]
