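# Physics asset for the female medium Jurassic Archaeology Summer cosmetic body, setting up secondary motion for its bows, necklace and ropes.
# Capsules on the pelvis, spine, left forearm twist and left hand form the body volumes.
# Constraint chains hang the A bows and necklace from spine_05, B bows from spine_01, and two ropes from the left hand.
# The necklace is restricted mostly to single-axis swing, narrowing down the chain.
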PhysicsAsset /BRCosmetics/Characters/Player/Female/Medium/Bodies/F_MED_Jurassic_Archaeology_Summer/Meshes/F_MED_Jurassic_Archaeology_Summer_Physics3
bodies {
 dny_bowA_l_01 [{shapeType=Capsule, length=2.4278, radius=0.5131, center=<-2.104 0.0514 -0.131>, rotation=<rot 1.3881 -91.2249 -88.3345>}]
 dny_bowA_l_02 [{shapeType=Capsule, length=1.9572, radius=0.5131, center=<-0.5638 0.251 -0.0798>, rotation=<rot -1.4404 -105.2031 -91.7049>}]
 dny_bowA_r_01 [{shapeType=Capsule, length=2.6118, radius=0.5131, center=<-2.0453 -0.088 -0.1196>, rotation=<rot -0.5201 91.7461 -89.2005>}]
 dny_bowA_r_02 [{shapeType=Capsule, length=1.8565, radius=0.5131, center=<-0.7716 0.1434 0.027>, rotation=<rot 0 -105.0409 -86.4735>}]
 dny_bowB_l_02 [{shapeType=Capsule, length=0.505, radius=0.505, rotation=<rot 0 -90 0>}]
 dny_bowB_l_03 [{shapeType=Capsule, length=0.505, radius=0.505, rotation=<rot 0 -90 0>}]
 dny_bowB_r_02 [{shapeType=Capsule, length=0.505, radius=0.505, rotation=<rot 0 -90 0>}]
 dny_bowB_r_03 [{shapeType=Capsule, length=0.505, radius=0.505, rotation=<rot 0 -90 0>}]
 dyn_necklace01 [{shapeType=Capsule, length=4.6342, radius=0.5001, center=<-2.426 0.0654 -0.0455>, rotation=<rot 88.8853 179.9957 179.9958>}]
 dyn_necklace02 [{shapeType=Capsule, length=1.8519, radius=1.2185, center=<-2.1654 -0.0403 -0.0256>, rotation=<rot 89.2678 179.9995 -178.0308>}]
 dyn_necklace03 [{shapeType=Capsule, length=0.8672, radius=0.8285, center=<-0.474 0.0489 0.0034>, rotation=<rot 89.5748 -0.0011 5.1911>}]
 dyn_necklace04 [{shapeType=Capsule, length=2.1434, radius=1.2407, center=<-2.1585 0.0831 0.0038>, rotation=<rot 0.0129 98.243 -90.2677>}]
 dyn_ropeA01 [{shapeType=Capsule, length=2.3641, radius=0.505, center=<-0.0425 -0.4913 0.1932>, rotation=<rot 0 177.2643 -81.6906>}]
 dyn_ropeA02 [{shapeType=Capsule, length=1.3231, radius=0.5, center=<0.1454 -1.0356 0.0264>, rotation=<rot 0 -168.788 -83.6428>}]
 dyn_ropeB01 [{shapeType=Capsule, length=2.0967, radius=0.505, center=<-0.0554 -0.8033 -0.1498>, rotation=<rot 88.1467 -140.2291 -51.6553>}]
 dyn_ropeB02 [{shapeType=Capsule, length=1.0868, radius=0.5, center=<-0.0232 -0.7619 -0.1286>, rotation=<rot 0 -173.9122 -81.6464>}]
 hand_l [{shapeType=Capsule, length=2.3233, radius=4.5237, center=<4.7368 -0.6578 -0.0943>, rotation=<rot 5.9174 -90.2278 86.4656>}]
 lowerarm_twist_01_l [{shapeType=Capsule, length=15.01, radius=3.1087, center=<-4.6101 0 0.2443>, rotation=<rot 86.9654 0.0032 0.0031>}]
 pelvis [{shapeType=Capsule, length=13.1009, radius=9.512, center=<4.5338 -1.2612 -0.2481>, rotation=<rot -2.4417 -90 2.0417>}]
 spine_01 [{shapeType=Capsule, length=3.8798, radius=7.6475, center=<8.8479 2.0168 0.0723>, rotation=<rot 0 -6.4479 -177.0717>}]
 spine_03 [{shapeType=Capsule, length=9.6058, radius=8.2983, center=<5.125 5.461 0.3517>, rotation=<rot 0.0855 -90.4754 -2.9988>}]
 spine_05 [
  {shapeType=Capsule, length=11.3324, radius=8.4279, center=<-1.8002 1.0164 -0.0085>, rotation=<rot 0 40.5822 -179.927>}
  {shapeType=Capsule, length=16.7461, radius=5.7297, center=<5.2462 -2.1692 0>}
 ]
}
constraints [
 {bone1Name=spine_01, bone2Name=pelvis, swing1LimitDegrees=45, swing2LimitDegrees=45, twistLimitDegrees=45}
 {bone1Name=spine_03, bone2Name=spine_01, swing1LimitDegrees=45, swing2LimitDegrees=45, twistLimitDegrees=45}
 {bone1Name=spine_05, bone2Name=spine_03, swing1LimitDegrees=45, swing2LimitDegrees=45, twistLimitDegrees=45}
 {bone1Name=dny_bowA_r_01, bone2Name=spine_05, swing1LimitDegrees=60, swing2LimitDegrees=10, twistLimitDegrees=0.5}
 {bone1Name=dny_bowA_r_02, bone2Name=dny_bowA_r_01, swing1LimitDegrees=40, swing2LimitDegrees=5, twistLimitDegrees=0.5}
 {bone1Name=dny_bowA_l_01, bone2Name=spine_05, swing1LimitDegrees=60, swing2LimitDegrees=10, twistLimitDegrees=0.5}
 {bone1Name=dny_bowA_l_02, bone2Name=dny_bowA_l_01, swing1LimitDegrees=40, swing2LimitDegrees=5, twistLimitDegrees=0.5}
 {bone1Name=dyn_necklace01, bone2Name=spine_05, swing1LimitDegrees=30, swing2LimitDegrees=0, twistLimitDegrees=0}
 {bone1Name=dyn_necklace02, bone2Name=dyn_necklace01, swing1LimitDegrees=20, swing2LimitDegrees=0, twistLimitDegrees=0}
 {bone1Name=dyn_necklace03, bone2Name=dyn_necklace02, swing1LimitDegrees=10, swing2LimitDegrees=0, twistLimitDegrees=0}
 {
  bone1Name=dyn_necklace04
  bone2Name=dyn_necklace03
  swing1LimitDegrees=10
  swing2LimitDegrees=1
  swing2Motion=Locked
  twistLimitDegrees=1
  twistMotion=Locked
 }
 {bone1Name=dyn_ropeA01, bone2Name=hand_l, swing1LimitDegrees=20, swing2LimitDegrees=1, twistLimitDegrees=20}
 {bone1Name=dyn_ropeA02, bone2Name=dyn_ropeA01, swing1LimitDegrees=45, swing2LimitDegrees=1, twistLimitDegrees=45}
 {bone1Name=dyn_ropeB01, bone2Name=hand_l, swing1LimitDegrees=60, swing2LimitDegrees=60, twistLimitDegrees=60}
 {bone1Name=dyn_ropeB02, bone2Name=dyn_ropeB01, swing1LimitDegrees=60, swing2LimitDegrees=60, twistLimitDegrees=60}
 {bone1Name=dny_bowB_l_03, bone2Name=spine_01, swing1LimitDegrees=90, swing2LimitDegrees=30, twistLimitDegrees=20}
 {bone1Name=dny_bowB_l_02, bone2Name=dny_bowB_l_03, swing1LimitDegrees=90, swing2LimitDegrees=30, twistLimitDegrees=20}
 {bone1Name=dny_bowB_r_03, bone2Name=spine_01, swing1LimitDegrees=90, swing2LimitDegrees=30, twistLimitDegrees=20}
 {bone1Name=dny_bowB_r_02, bone2Name=dny_bowB_r_03, swing1LimitDegrees=90, swing2LimitDegrees=30, twistLimitDegrees=20}
]
